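# Translucent two-sided effect material for the Ninja Crescent Kick trait, built for mesh particles and trails.
# Emissive layers tinted panning noise with two scrolling beam textures, with dynamic parameters varying beam scroll timing and blending toward a flat color.
# The ForUseWithRibbons switch selects the UV channel and the mask: beam texture times particle alpha for ribbons, otherwise Fresnel, sphere mask and a sharpened band moved by the Sweep dynamic parameter.
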Material /Game/CharClasses/Ninja/Traits/CrescentKick/Materials/M_Ninja_CrescentKick_02
graph {
 properties {
  bEnableResponsiveAA=true
  bUsedWithBeamTrails=true
  bUsedWithMeshParticles=true
  bUsedWithNiagaraMeshParticles=true
  blendMode=BLEND_Translucent
  pixelDepthOffsetMode=PDOM_Legacy
  translucencyLightingMode=TLM_VolumetricPerVertexDirectional
  twoSided=true
  usedShadingModels=Default Lit
 }
 nodes [
  n1 = TextureCoordinate() {coordinateIndex=2, uTiling=0.77, vTiling=2}
  n2 = TextureCoordinate() {uTiling=0.77, vTiling=2}
  n3 = TextureSample(UVs=Panner(Coordinate=StaticSwitchParameter(True=@n1, False=@n2) {parameterName=ForUseWithRibbons}) {bFractionalPart=true, speedX=-2.1892, speedY=0.6}) {texture=&T_NoisePack_01}
  n4 = Multiply(
   A=Add(A=Power(Base=@n3.G) {constExponent=4}, B=ScalarParameter() {defaultValue=0.01, parameterName=BG Bias})
   B=VectorParameter() {parameterName=Emissive Color B, defaultValue=<rgba 0.2 3.5 20 0>}.RGB
  )
  n5 = StaticSwitchParameter(True=TextureCoordinate() {coordinateIndex=2, uTiling=1.22}, False=TextureCoordinate() {uTiling=1.22}) {parameterName=ForUseWithRibbons}
  n6 = Constant2Vector() {g=0.09}
  n7 = Multiply(A=DynamicParameter() {defaultValue=<rgba 1 0 1 0>, paramNames [RandomScale,Sweep,Unused,Add]}.RandomScale, B=Time())
  n8 = TextureSample(UVs=Panner(Coordinate=Add(A=@n5, B=@n6), Time=@n7) {bFractionalPart=true, speedX=-3.113}) {samplerType=SAMPLERTYPE_LinearColor, texture=/Game/Effects/Fort_Effects/Textures/Tile/T_Raygun_Beam_01}
  n9 = StaticSwitchParameter(True=TextureCoordinate() {coordinateIndex=2, uTiling=1.418}, False=TextureCoordinate() {uTiling=1.418}) {parameterName=ForUseWithRibbons}
  n10 = Constant2Vector() {g=-0.04}
  n11 = TextureSample(UVs=Panner(Coordinate=Add(A=@n9, B=@n10), Time=@n7) {bFractionalPart=true, speedX=-3.84}) {samplerType=SAMPLERTYPE_LinearColor, texture=&T_Raygun_Beam_01}
  n12 = VectorParameter() {parameterName=Emissive Color A, defaultValue=<rgba 1.575 7.1687 15 0>}
  n13 = Multiply(A=Add(A=Multiply(A=@n8.G, B=@n11.G), B=@n11.B), B=@n12.RGB)
  MaterialFunctionCall(Emissive=LinearInterpolate(A=Add(A=@n4, B=@n13), B=@n12.RGB, Alpha=DynamicParameter() {defaultValue=<rgba 1 0 1 0>, paramNames [RandomScale,Sweep,Unused,Add]}.Add)) {materialFunction=&MF_FX_EmissiveAdjustments}
  n14 = MaterialFunctionCall(Exponent=ScalarParameter() {defaultValue=4, parameterName=Fresnel Exponent}) {materialFunction=&MF_Fresnel_FX}
  n15 = SphereMask(
   A=TextureCoordinate()
   B=Constant() {r=0.5}
   Radius=ScalarParameter() {defaultValue=0.5, parameterName=Mask Radius}
   Hardness=ScalarParameter() {defaultValue=0.6, parameterName=Mask Hardness}
  )
  n16 = ComponentMask(None=Add(A=TextureCoordinate(), B=DynamicParameter() {defaultValue=<rgba 1 0 1 0>, paramNames [RandomScale,Sweep,Unused,Add]}.Sweep)) {r=true}
  n17 = OneMinus(None=@n16)
  n18 = Clamp(None=Multiply(A=Multiply(A=@n16, B=@n17)) {constB=4})
  n19 = ScalarParameter() {defaultValue=25, parameterName=Sweep Exponent}
  StaticSwitchParameter(True=Multiply(A=@n11.R, B=ParticleColor().A), False=Multiply(A=Multiply(A=@n14.Fresnel, B=@n15), B=Power(Base=@n18, Exp=@n19))) {parameterName=ForUseWithRibbons}
 ]
}
